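# CustomDepthSort is a material function in the Water debug materials that outputs a binary mask from comparing pixel depth against custom depth.
# It scales the CustomDepth scene texture by a tunable offset parameter, slightly above one by default, and subtracts that from pixel depth.
# Ceil, Saturate and OneMinus turn the difference into a hard mask: white at or in front of the scaled custom depth, black behind it.
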MaterialFunction /Water/Materials/Debug/CustomDepthSort
graph {nodes [
  n1 = ComponentMask(None=SceneTexture() {sceneTextureId=PPI_CustomDepth}.Color) {r=true}
  n2 = ScalarParameter() {defaultValue=1.01, group=Technical, parameterName=Custom Depth Comparison offset}
  n3 = Ceil(None=Subtract(A=PixelDepth().R, B=Multiply(A=@n1, B=@n2)))
  FunctionOutput(None=OneMinus(None=Saturate(None=@n3)))
 ]}
