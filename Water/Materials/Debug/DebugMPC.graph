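# Unlit debug material for inspecting values in the MPC_Water material parameter collection.
# It reads TerrainMinZ, TerrainMaxZ and the XY of SimLocation and appends them into one four-component vector.
# That vector feeds the engine DebugFloat4Values function, which renders numbers as text.
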Material /Water/Materials/Debug/DebugMPC
graph {
 properties {pixelDepthOffsetMode=PDOM_Legacy, shadingModel=MSM_Unlit, usedShadingModels=Unlit}
 nodes [
  n1 = CollectionParameter() {collection=/Water/Materials/MPC/MPC_Water, parameterName=TerrainMinZ}
  n2 = CollectionParameter() {collection=&MPC_Water, parameterName=TerrainMaxZ}
  n3 = CollectionParameter() {collection=&MPC_Water, parameterName=SimLocation}
  MaterialFunctionCall!(
   Vector4=AppendVector(A=AppendVector(A=@n1, B=@n2), B=ComponentMask(None=@n3) {g=true, r=true})
   MaximumNumberOfDigits
   UVs
   DebugTextLocation RG_UpperRight BA_LowerLeft
   Component Spacing
  ) {materialFunction=/Engine/Functions/Engine_MaterialFunctions02/Utility/DebugFloat4Values}
 ]
}
