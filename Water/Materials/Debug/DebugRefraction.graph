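# Unlit translucent debug material for inspecting screen-space refraction math in the Water content.
# A custom HLSL node turns a normal map into view-space UV distortion, corrected for FOV and view size and faded by scene depth against pixel depth.
# Side branches hold scaled absolute distortion and banded scene or pixel depth; no material output connection is shown.
# The aspect ratio input divides one view-size component by itself.
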Material /Water/Materials/Debug/DebugRefraction
graph {
 properties {blendMode=BLEND_Translucent, pixelDepthOffsetMode=PDOM_Legacy, shadingModel=MSM_Unlit, usedShadingModels=Unlit}
 nodes [
  MaterialExpressionTextureSample_0 = TextureSample() {samplerType=SAMPLERTYPE_Normal, texture=/Engine/Functions/Engine_MaterialFunctions02/ExampleContent/Textures/newhplaneNormalsMap}
  Custom!(Normal=@MaterialExpressionTextureSample_0.RGB) {
   code="float3 RefractNormal = normalize(TransformWorldVectorToView(Normal));\r\n\r\n\r\nreturn RefractNormal;"
   description=Get Distorted UV
   outputType=CMOT_Float2
   inputs [{
     inputName=Normal
     input {expression=&self:MaterialExpressionTextureSample_0, mask=1, maskB=1, maskG=1, maskR=1}
    }]
  }
  MaterialExpressionSceneDepth_1 = SceneDepth()
  n1 = PixelDepth()
  n2 = ScalarParameter() {parameterName=use pd}
  Frac(None=Divide(A=LinearInterpolate(A=@MaterialExpressionSceneDepth_1.R, B=@n1.R, Alpha=@n2)) {constB=1000})
  n3 = ViewProperty() {property=MEVP_ViewSize}
  MaterialFunctionCall(Float2=@n3.Property) {materialFunction=&BreakOutFloat2Components}
  MaterialExpressionScreenPosition_2 = ScreenPosition()
  MaterialExpressionAppendVector_0 = AppendVector(A=ScalarParameter() {defaultValue=1.3, parameterName=TestRefraction}, B=ScalarParameter() {defaultValue=0.1, parameterName=RefractionBias})
  n4 = Divide(B=ViewProperty() {property=MEVP_TanHalfFieldOfView}.Property)
  n5 = MaterialFunctionCall(Float2=@n3.Property) {materialFunction=&BreakOutFloat2Components}
  MaterialExpressionAppendVector_2 = AppendVector(A=AppendVector(A=ComponentMask(None=@n4) {r=true}, B=Divide(A=@n5.R, B=@n5.R)), B=ComponentMask(None=@n3.Property) {g=true, r=true})
  n6 = Custom!(
   Normal=@MaterialExpressionTextureSample_0.RGB
   ViewUV=@MaterialExpressionScreenPosition_2.ViewportUV
   RefractionAndBias=@MaterialExpressionAppendVector_0
   SceneDepth=@MaterialExpressionSceneDepth_1.R
   DistortionParameters=@MaterialExpressionAppendVector_2
  ) {
   code="\r\n\r\n\r\nfloat3 ViewVertexNormal  = normalize(TransformWorldVectorToView( float3(0, 0, 1) ) );\r\n\r\nfloat3 ViewNormal = normalize(TransformWorldVectorToView(Normal));\r\n\r\nfloat2 ViewportUVDistortion = (ViewVertexNormal.xy - ViewNormal.xy) * (RefractionAndBias.x - 1.0);\r\n\r\nfloat2 BufferUVDistortion = ViewportUVDistortion * ResolvedView.ViewSizeAndInvSize.xy * ResolvedView.BufferSizeAndInvSize.zw;\r\n\r\nfloat InvTanHalfFov = DistortionParameters.x;\r\nfloat Ratio = DistortionParameters.y;\r\n\r\nfloat2 FovFix = float2(InvTanHalfFov, Ratio*InvTanHalfFov);\r\n\r\nconst float OffsetFudgeFactor = 0.00023;\r\nBufferUVDistortion *= DistortionParameters.zw * float2(OffsetFudgeFactor, -OffsetFudgeFactor) * FovFix;\r\n\r\n\r\n\r\n\tfloat Bias = -RefractionAndBias.y;\r\n\tfloat Range = clamp(abs(Bias * 0.5f), 0, 50);\r\n\tfloat Z = SceneDepth;\r\n\tfloat ZCompare = Parameters.ScreenPosition.w;\r\n\tfloat InvWidth = 1.0f / max(1.0f, Range);\r\n\tBufferUVDistortion *= saturate((Z - ZCompare) * InvWidth + Bias);\r\n\r\n\t//Scale up for better precision in low/subtle refractions at the expense of artefacts at higher refraction.\r\n\tstatic const half DistortionScaleBias = 4.0f;\r\n\tBufferUVDistortion *= DistortionScaleBias;\r\n\t\r\n\treturn BufferUVDistortion;"
   description=Get Distorted UV
   outputType=CMOT_Float2
   inputs [
    {
     inputName=Normal
     input {expression=&self:MaterialExpressionTextureSample_0, mask=1, maskB=1, maskG=1, maskR=1}
    }
    {inputName=ViewUV, input {expression=&self:MaterialExpressionScreenPosition_2}}
    {inputName=RefractionAndBias, input {expression=&self:MaterialExpressionAppendVector_0}}
    {inputName=SceneDepth, input {expression=&self:MaterialExpressionSceneDepth_1, mask=1, maskR=1}}
    {inputName=DistortionParameters, input {expression=&self:MaterialExpressionAppendVector_2}}
   ]
  }
  Multiply(A=Abs(None=@n6), B=ScalarParameter() {defaultValue=1, parameterName=mult})
  SceneDepthWithoutWater!(UVs)
 ]
}
