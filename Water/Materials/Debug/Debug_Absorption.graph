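# Translucent unlit debug material for water absorption, tinting the scene behind the surface by the depth of material looked through.
# Thickness is scene depth minus pixel depth, clamped by a Max, then scaled by the inverted per-channel Reciprocal Meters vector and the DDD scalar.
# A Custom node applies exp(-x) for Beer-Lambert style falloff, and the result multiplies SceneColor.
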Material /Water/Materials/Debug/Debug_Absorption
graph {
 properties {blendMode=BLEND_Translucent, pixelDepthOffsetMode=PDOM_Legacy, shadingModel=MSM_Unlit, usedShadingModels=Unlit}
 nodes [
  n1 = SceneDepth()
  n2 = PixelDepth()
  n3 = VectorParameter() {parameterName=Reciprocal Meters, defaultValue=<rgba 1.6667 20 227.2727 0>}
  MaterialExpressionMultiply_6 = Multiply(
   A=Max(A=Subtract(A=@n1.R, B=@n2.R)) {constB=0}
   B=Multiply(A=Divide(B=@n3.RGB) {desc=Attenuation Length}, B=ScalarParameter() {defaultValue=0.01, parameterName=DDD}) {constB=100}
  )
  n4 = Custom!(x=@MaterialExpressionMultiply_6) {
   code="return exp(-x);"
   description=exp
   inputs [{inputName=x, input {expression=&self:MaterialExpressionMultiply_6}}]
  }
  Multiply(A=SceneColor!(UVs).RGB, B=@n4)
 ]
}
