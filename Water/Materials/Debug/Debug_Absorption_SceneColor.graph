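# Unlit debug material in the Water content for inspecting an absorption value.
# A Custom HLSL node computes exponential falloff, exp(-x), from the "Attenuation length meters" scalar parameter, using it directly as the exponent.
# The result feeds the DebugScalarValues function as the number to display.
# Despite the SceneColor name, the graph contains no scene color sample and shows no link to a material output.
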Material /Water/Materials/Debug/Debug_Absorption_SceneColor
graph {
 properties {pixelDepthOffsetMode=PDOM_Legacy, shadingModel=MSM_Unlit, usedShadingModels=Unlit}
 nodes [
  MaterialExpressionScalarParameter_0 = ScalarParameter() {defaultValue=1, parameterName=Attenuation length meters}
  n1 = Custom!(x=@MaterialExpressionScalarParameter_0) {
   code="return exp(-x);"
   outputType=CMOT_Float1
   inputs [{inputName=x, input {expression=&self:MaterialExpressionScalarParameter_0}}]
  }
  MaterialFunctionCall(Number=@n1) {materialFunction=&DebugScalarValues}
 ]
}
