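# Debug material in the Water plugin for inspecting Z height of water waves, with no material output wired in the graph shown.
# One branch samples FoamRT around the MPC_Water SimLocation, scaled by WaveCaptureSize, and turns its green channel minus world Z into a scaled blue vector.
# Another does likewise with the Z offset from ComputeGerstnerWaves.
# Leftover experiments include an unused GetGerstnerWaves call and a saturated green-channel offset.
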Material /Water/Materials/Debug/Debug_Z_Height
graph {
 properties {pixelDepthOffsetMode=PDOM_Legacy, usedShadingModels=Default Lit}
 nodes [
  n1 = MaterialFunctionCall() {materialFunction=/Water/Materials/Functions/GetWaterTime}
  MaterialFunctionCall!(WaveParameterRT, Override Position, Time=@n1.WaterTime) {materialFunction=/Water/Caustics/Materials/Functions/GetGerstnerWaves}
  n2 = Subtract(
   A=ComponentMask(None=WorldPosition().XYZ) {g=true, r=true}
   B=ComponentMask(None=CollectionParameter() {collection=/Water/Materials/MPC/MPC_Water, parameterName=SimLocation}) {g=true, r=true}
  )
  n3 = CollectionParameter() {collection=&MPC_Water, parameterName=WaveCaptureSize}
  n4 = TextureSample(UVs=Add(A=Divide(A=@n2, B=@n3)) {constB=0.5}) {samplerType=SAMPLERTYPE_LinearColor, texture=/Water/FluidSimulation/WaveFoam/FoamRT}
  n5 = ComponentMask(None=WorldPosition().XYZ) {b=true}
  n6 = Constant3Vector() {constant=<rgba 0 0 1 0>}
  n7 = ScalarParameter() {defaultValue=1, parameterName=V}
  Multiply(A=Multiply(A=Subtract(A=@n4.G, B=@n5), B=@n6), B=@n7)
  n8 = MaterialFunctionCall!(Water Body Index=ScalarParameter() {parameterName=WB}, Override Position, Time=@n1.WaterTime, Use Depth Attenuation=StaticBool()) {materialFunction=/Water/Materials/Functions/ComputeGerstnerWaves}
  Multiply(A=Multiply(A=@n6, B=ComponentMask(None=@n8.WPO) {b=true}), B=@n7)
  Saturate(None=Add(A=@n4.G, B=ScalarParameter() {defaultValue=256, parameterName=asdf}))
 ]
}
