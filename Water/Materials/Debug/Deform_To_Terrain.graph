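# Unlit debug material for the Water plugin, usable on spline meshes, built to deform geometry to the terrain height.
# WaterBodyData supplies Terrain Z Height, which feeds DebugScalarValues as a number over tiled world-XY UVs scaled by the XY parameter.
# The same height plus a Water Bias parameter, minus world Z, is multiplied by a Z-only vector, forming a vertical offset onto the biased terrain.
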Material /Water/Materials/Debug/Deform_To_Terrain
graph {
 properties {bUsedWithSplineMeshes=true, pixelDepthOffsetMode=PDOM_Legacy, shadingModel=MSM_Unlit, usedShadingModels=Unlit}
 nodes [
  n1 = MaterialFunctionCall!(Position) {materialFunction=/Water/Materials/Functions/WaterBodyData}
  n2 = ScalarParameter() {defaultValue=512, parameterName=XY}
  MaterialFunctionCall(
   Number=@n1.Terrain Z Height
   UVs=Frac(None=MaterialFunctionCall!(Scale=@n2) {materialFunction=/Engine/Functions/Engine_MaterialFunctions01/Coordinates/WorldPosition-XY}.Result)
  ) {materialFunction=&DebugScalarValues}
  n3 = ScalarParameter() {parameterName=Water Bias}
  n4 = WorldPosition()
  Multiply(A=Subtract(A=Add(A=@n1.Terrain Z Height, B=@n3), B=ComponentMask(None=@n4.XYZ) {b=true}), B=Constant3Vector() {constant=<rgba 0 0 1 0>})
 ]
}
