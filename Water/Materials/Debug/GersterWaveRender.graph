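# Debug material that evaluates Gerstner waves from the DefaultWaveSpectrumRT texture at world position, with parameters to freeze or manually set time.
# A Custom HLSL struct sums 16 waves stored as texel pairs, then packs offset into integer and normal into fractional parts.
# Two Custom nodes unpack these, though normal unpacking scales by 2.2 versus 2.02 in packing.
# A separate GetGerstnerWaves function call takes time from GetWaterTime.
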Material /Water/Materials/Debug/GersterWaveRender
graph {
 properties {pixelDepthOffsetMode=PDOM_Legacy, usedShadingModels=Default Lit}
 nodes [
  MaterialExpressionTextureObject_0 = TextureObject() {samplerType=SAMPLERTYPE_LinearColor, texture=/Water/Textures/DefaultWaveSpectrumRT}
  MaterialExpressionWorldPosition_0 = WorldPosition() {worldPositionShaderOffset=WPT_ExcludeAllShaderOffsets}
  MaterialExpressionLinearInterpolate_0 = LinearInterpolate(A=Time(), B=ScalarParameter() {parameterName=Manual Time}, Alpha=ScalarParameter() {parameterName=Freeze Time})
  MaterialExpressionCustom_0 = Custom!(
   WaveParameterRT=@MaterialExpressionTextureObject_0
   WorldPosition=@MaterialExpressionWorldPosition_0.XYZ
   Time=@MaterialExpressionLinearInterpolate_0
  ) {
   code="##define Gravity 980\r\n##define SOLVE_NORMAL_Z 1\r\n##define SteepnessThreshold 25\r\n\r\nstruct WaveParams\r\n{\r\n\tfloat2 Direction;\r\n\tfloat Wavelength;\r\n\tfloat Amplitude;\r\n\tfloat Steepness;\r\n};\r\n\r\nstruct WaveOutput\r\n{\r\n\tfloat3 Normal;\r\n\tfloat3 WPO;\r\n};\r\n\r\nstruct GerstnerWaveRenderer\r\n{\r\n\r\n\tTexture2D WaveParamsRT;\r\n\tSamplerState WaveParamsRTSampler;\r\n\tfloat2 WaveParamRTSize;\r\n\tint MaxWaves;\r\n\tfloat2 WorldPos;\r\n\tfloat Time;\r\n\t\r\n\tWaveParams GetWaveRTData(int inWaveIndex, int inWaterBodyIndex)\r\n\t{\r\n\t\tWaveParams OutWaveParams;\r\n\t\t\r\n\t\tfloat2 WaveUV;\r\n\t\tfloat WaveUV_b;\r\n\t\t\r\n\t\tinWaveIndex *= 2;\r\n\t\t\r\n\t\tWaveUV.x = inWaveIndex / WaveParamRTSize.x;\r\n\t\tWaveUV.y = inWaterBodyIndex / WaveParamRTSize.y;\r\n\t\tWaveUV += 0.5 / WaveParamRTSize;\r\n\t\t\r\n\t\tWaveUV_b = WaveUV.x + ( 1 / WaveParamRTSize.x);\r\n\t\t\r\n\t\tOutWaveParams.Direction = WaveParamsRT.SampleLevel(WaveParamsRTSampler, WaveUV, 0).rg;\r\n\t\t\r\n\t\tfloat3 TempParams = WaveParamsRT.SampleLevel(WaveParamsRTSampler, float2(WaveUV_b, WaveUV.y), 0);\r\n\t\t\r\n\t\tOutWaveParams.Wavelength = TempParams.x;\r\n\t\tOutWaveParams.Amplitude = TempParams.y;\r\n\t\tOutWaveParams.Steepness = TempParams.z;\r\n\t\t\r\n\t\treturn OutWaveParams;\r\n\t}\r\n\t\r\n\tWaveOutput AddWaves(WaveOutput inWaveA, WaveOutput inWaveB)\r\n\t{\r\n\t\tinWaveA.Normal += inWaveB.Normal;\r\n\t\tinWaveA.WPO += inWaveB.WPO;\r\n\t\t\r\n\t\treturn inWaveA;\r\n\t}\r\n\t\r\n\tfloat3 FinalizeNormal(float3 inNormal)\r\n\t{\r\n\t\treturn normalize(float3(inNormal.xy, 1-inNormal.z));\r\n\t}\r\n\t\r\n\tfloat3 PackNormalAndWPO(WaveOutput inWave)\r\n\t{\r\n\t\tfloat3 packedoutput = 0;\r\n\t\t\r\n\t\tpackedoutput = floor(inWave.WPO * 1000);\r\n\t\tpackedoutput += ((inWave.Normal + 1.01) / 2.02);\r\n\t\r\n\t\treturn packedoutput;\r\n\t}\r\n\t\r\n\tfloat3 UnpackWaveNormal(float inPackedWave)\r\n\t{\r\n\t\tfloat3 outnormal = frac(inPackedWave);\r\n\t\toutnormal *= 2.2;\r\n\t\toutnormal -= 1.01;\r\n\t\treturn outnormal;\r\n\t}\r\n\t\r\n\tfloat3 UnpackWaveWPO(float inPackedWave)\r\n\t{\r\n\t\tfloat3 outWPO;\r\n\t\toutWPO = floor(inPackedWave) / 1000;\r\n\t\treturn outWPO;\r\n\t}\r\n\t\r\n\tfloat2 GetWavePositionAndDispersion(WaveParams inWaveParams)\r\n\t{\r\n\t\tfloat dispersion = 2 * PI / inWaveParams.Wavelength;\r\n\t\tfloat2 wavevector = inWaveParams.Direction * dispersion;\r\n\t\tfloat wavespeed = sqrt(dispersion * Gravity);\r\n\t\tfloat wavetime = wavespeed * Time;\r\n\t\t\r\n\t\tfloat wavepos = dot(WorldPos, wavevector) - wavetime;\r\n\t\t\r\n\t\treturn float2(wavepos, dispersion);\r\n\t}\r\n\t\r\n\tWaveOutput ComputeSingleWaveVectorized(float inWaveSin, float inWaveCos, float inDispersion, WaveParams inWaveParams)\r\n\t{\r\n\t\tWaveOutput OutWave;\r\n\t\t\r\n\t\tfloat wKA = inWaveParams.Amplitude * inDispersion;\r\n\t\t\r\n\t\tfloat q = inWaveParams.Steepness / wKA;\r\n\t\t\r\n\t\tOutWave.Normal.xy = inWaveSin * wKA * inWaveParams.Direction;\r\n\t\t\r\n\t\t##if SOLVE_NORMAL_Z\r\n\t\t\tOutWave.Normal.z = inWaveCos * wKA * q;\r\n\t\t##else\r\n\t\t\tOutWave.Normal.z = 0;\r\n\t\t##endif\r\n\r\n\t\tOutWave.WPO.xy = -q * inWaveSin * inWaveParams.Direction * inWaveParams.Amplitude;\r\n\t\tOutWave.WPO.z = inWaveCos * inWaveParams.Amplitude;\r\n\r\n\t\treturn OutWave;\r\n\t}\r\n\t\r\n\tWaveOutput GetSingleGerstnerWave(int inWaveIndex, int inWaterBodyIndex)\r\n\t{\r\n\t\tWaveOutput OutWave;\r\n\t\t\r\n\t\tWaveParams CurrentWave;\r\n\t\tCurrentWave = GetWaveRTData(inWaveIndex, inWaterBodyIndex);\r\n\t\t\r\n\t\tfloat dispersion = 2 * PI / CurrentWave.Wavelength;\r\n\t\tfloat2 wavevector = CurrentWave.Direction * dispersion;\r\n\t\tfloat wavespeed = sqrt(dispersion * Gravity);\r\n\t\tfloat wavetime = wavespeed * Time;\r\n\r\n\t\tfloat wavepos = dot(WorldPos, wavevector) - wavetime;\r\n\t\t\r\n\t\tfloat wavesin = sin(wavepos);\r\n\t\tfloat wavecos = cos(wavepos);\r\n\t\t\r\n\t\tfloat wKA = CurrentWave.Amplitude * dispersion;\r\n\t\t\r\n\t\tfloat q = CurrentWave.Steepness / wKA;\r\n\t\t\r\n\t\tOutWave.Normal.xy = wavesin * wKA * CurrentWave.Direction;\r\n\t\t\r\n\t\t##if SOLVE_NORMAL_Z\r\n\t\t\tOutWave.Normal.z = wavecos *  CurrentWave.Steepness * saturate( (CurrentWave.Amplitude * SteepnessThreshold) / CurrentWave.Wavelength );\r\n\t\t##else\r\n\t\t\tOutWave.Normal.z = 0;\r\n\t\t##endif\r\n\r\n\t\tOutWave.WPO.xy = -q * wavesin * CurrentWave.Direction *  CurrentWave.Amplitude;\r\n\t\tOutWave.WPO.z = wavecos * CurrentWave.Amplitude;\r\n\r\n\t\treturn OutWave;\r\n\t\t\r\n\t}\r\n\t\r\n\tWaveOutput GetAllGerstnerWaves(int inWaterBodyIndex)\r\n\t{\r\n\t\t\r\n\t\tWaveOutput OutWaves;\r\n\t\tWaveOutput CurrentWave;\r\n\t\t\r\n\t\tint MaxWaveRange = MaxWaves - 1;\r\n\t\t\r\n\t\tfor (int i = 0; i <=  MaxWaveRange; i++)\r\n\t\t{\r\n\t\t\tCurrentWave = GetSingleGerstnerWave(i, inWaterBodyIndex);\r\n\t\t\tOutWaves = AddWaves(OutWaves, CurrentWave);\r\n\t\t}\r\n\t\t\r\n\t\t//The Normal B channel must be inverted after combining waves\r\n\t\tOutWaves.Normal = FinalizeNormal(OutWaves.Normal);\r\n\t\treturn OutWaves;\r\n\t}\r\n\t\r\n\tWaveOutput GetAllGerstnerWavesVectorized(int inWaterBodyIndex)\r\n\t{\r\n\t\t\r\n\t\tWaveOutput OutWaves;\r\n\t\tWaveOutput CurrentWave;\r\n\t\t\r\n\t\tint MaxWaveRange = (MaxWaves / 4 )- 1;\r\n\t\tfor(int i = 0; i <= MaxWaveRange; i++)\r\n\t\t{\r\n\t\t\tWaveParams WaveA;\r\n\t\t\tWaveParams WaveB;\r\n\t\t\tWaveParams WaveC;\r\n\t\t\tWaveParams WaveD;\r\n\t\t\t\r\n\t\t\tfloat2 WaveAPosAndDispersion = GetWavePositionAndDispersion(WaveA);\r\n\t\t\tfloat2 WaveBPosAndDispersion = GetWavePositionAndDispersion(WaveB);\r\n\t\t\tfloat2 WaveCPosAndDispersion = GetWavePositionAndDispersion(WaveC);\r\n\t\t\tfloat2 WaveDPosAndDispersion = GetWavePositionAndDispersion(WaveD);\r\n\t\t\t\r\n\t\t\tfloat4 WaveSines = 0;\r\n\t\t\tfloat4 WaveCosines = 0;\r\n\t\t\t\r\n\t\t\tWaveSines = sin( float4( WaveAPosAndDispersion.x, WaveBPosAndDispersion.x, WaveCPosAndDispersion.x, WaveDPosAndDispersion.x));\r\n\t\t\tWaveCosines = cos( float4( WaveAPosAndDispersion.x, WaveBPosAndDispersion.x, WaveCPosAndDispersion.x, WaveDPosAndDispersion.x));\r\n\t\t\t\r\n\t\t\t//sincos( float4( WaveAPosAndDispersion.x, WaveBPosAndDispersion.x, WaveCPosAndDispersion.x, WaveDPosAndDispersion.x), WaveSines, WaveCosines);\r\n\t\t\t\r\n\t\t\tCurrentWave = ComputeSingleWaveVectorized(WaveSines.x, WaveCosines.x, WaveAPosAndDispersion.y, WaveA);\r\n\t\t\tOutWaves = AddWaves(OutWaves, CurrentWave);\r\n\t\t\t\r\n\t\t\tCurrentWave = ComputeSingleWaveVectorized(WaveSines.y, WaveCosines.y, WaveBPosAndDispersion.y, WaveB);\r\n\t\t\tOutWaves = AddWaves(OutWaves, CurrentWave);\r\n\t\t\t\r\n\t\t\tCurrentWave = ComputeSingleWaveVectorized(WaveSines.z, WaveCosines.z, WaveCPosAndDispersion.y, WaveC);\r\n\t\t\tOutWaves = AddWaves(OutWaves, CurrentWave);\r\n\t\t\t\r\n\t\t\tCurrentWave = ComputeSingleWaveVectorized(WaveSines.a, WaveCosines.a, WaveDPosAndDispersion.y, WaveD);\r\n\t\t\tOutWaves = AddWaves(OutWaves, CurrentWave);\r\n\t\t\t\r\n\t\t}\r\n\r\n\t\t//The Normal B channel must be inverted after combining waves\r\n\t\tOutWaves.Normal = FinalizeNormal(OutWaves.Normal);\r\n\t\treturn OutWaves;\r\n\t}\r\n\t\r\n\tWaveOutput GetRangeGerstnerWaves(int inMinWaveIndex, int inMaxWaveIndex, int inWaterBodyIndex)\r\n\t{\r\n\t\r\n\t\tWaveOutput OutWaves;\r\n\t\tWaveOutput CurrentWave;\r\n\t\t\r\n\t\tint MaxWaveRange = min(inMaxWaveIndex, MaxWaves - 1);\r\n\t\t\r\n\t\tfor (int i = inMinWaveIndex; i <= MaxWaveRange; i++)\r\n\t\t{\r\n\t\t\tCurrentWave = GetSingleGerstnerWave(i, inWaterBodyIndex);\r\n\t\t\tOutWaves = AddWaves(OutWaves, CurrentWave);\r\n\t\t}\r\n\t\t\r\n\t\t//The Normal B channel must be inverted after combining waves\r\n\t\tOutWaves.Normal = FinalizeNormal(OutWaves.Normal);\r\n\t\t\r\n\t\treturn OutWaves;\r\n\t}\r\n};\r\n\r\nGerstnerWaveRenderer WaveRenderer;\r\nWaveRenderer.WaveParamsRT = WaveParameterRT;\r\nWaveRenderer.WaveParamsRTSampler = WaveParameterRTSampler;\r\nWaveRenderer.WaveParamRTSize = float2(32,1);\r\nWaveRenderer.MaxWaves = 16;\r\nWaveRenderer.WorldPos = WorldPosition;\r\nWaveRenderer.Time = Time;\r\n\r\n\r\nWaveOutput Waves;\r\n\r\n##if 1\r\n\tWaves = WaveRenderer.GetAllGerstnerWaves(0);\r\n\treturn WaveRenderer.PackNormalAndWPO(Waves);\r\n\t\r\n##else\r\n\tWaves = WaveRenderer.GetSingleGerstnerWave(0, 0);\r\n\tWaves.Normal = WaveRenderer.FinalizeNormal(Waves.Normal);\r\n\treturn WaveRenderer.PackNormalAndWPO(Waves);\r\n\r\n##endif\r\n"
   inputs [
    {inputName=WaveParameterRT, input {expression=&self:MaterialExpressionTextureObject_0}}
    {inputName=WorldPosition, input {expression=&self:MaterialExpressionWorldPosition_0}}
    {inputName=Time, input {expression=&self:MaterialExpressionLinearInterpolate_0}}
   ]
  }
  Custom!(inPackedWave=@MaterialExpressionCustom_0) {
   code="\t\tfloat3 outWPO;\r\n\t\toutWPO = floor(inPackedWave) / 1000;\r\n\t\treturn outWPO;"
   description=UnpackWPO
   inputs [{inputName=inPackedWave, input {expression=&self:MaterialExpressionCustom_0}}]
  }
  Custom(inPackedWave=@MaterialExpressionCustom_0) {
   code="\t\tfloat3 outnormal = frac(inPackedWave);\r\n\t\toutnormal *= 2.2;\r\n\t\toutnormal -= 1.01;\r\n\t\treturn outnormal;"
   description=UnpackNormal
   inputs [{inputName=inPackedWave, input {expression=&self:MaterialExpressionCustom_0}}]
  }
  MaterialFunctionCall!(WaveParameterRT, Override Position, Time=MaterialFunctionCall() {materialFunction=/Water/Materials/Functions/GetWaterTime}.WaterTime) {materialFunction=/Water/Caustics/Materials/Functions/GetGerstnerWaves}
  VectorParameter() {parameterName=C, defaultValue=<rgba 0 0.2 0.5 1>}
 ]
}
