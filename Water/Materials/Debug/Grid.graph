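# Debug grid material in the Water plugin, built on the Default Lit shading model to show a world-aligned grid.
# The WorldPosition-XY function turns world position into UVs, with scalar parameter T as its scale, so the pattern is independent of mesh UVs.
# Those UVs sample the engine's 1x1 grid debug texture, and a small constant is added to its RGB to lift the dark areas.
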Material /Water/Materials/Debug/Grid
graph {
 properties {pixelDepthOffsetMode=PDOM_Legacy, usedShadingModels=Default Lit}
 nodes [
  n1 = ScalarParameter() {defaultValue=1024, parameterName=T}
  Add(A=TextureSample(UVs=MaterialFunctionCall!(Scale=@n1) {materialFunction=/Engine/Functions/Engine_MaterialFunctions01/Coordinates/WorldPosition-XY}.Result) {texture=/Engine/EngineDebugMaterials/T_1x1_Grid}.RGB) {constB=0.1}
 ]
}
