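# Unlit debug material in the Water content that tests a parallax-style directional blur on the engine's example flipbook texture.
# A tangent-space camera vector, with XY divided by Z, is scaled and offset by parameters to form a UV step direction.
# A Custom HLSL loop averages mip-zero samples along that direction, with parameters for distance and step count.
# Despite the name, it performs no height test.
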Material /Water/Materials/Debug/Parallax
graph {
 properties {pixelDepthOffsetMode=PDOM_Legacy, shadingModel=MSM_Unlit, usedShadingModels=Unlit}
 nodes [
  MaterialExpressionTextureObject_1 = TextureObject() {texture=/Engine/Functions/Engine_MaterialFunctions02/ExampleContent/Textures/flipbook}
  MaterialExpressionTextureCoordinate_0 = TextureCoordinate()
  n1 = Transform(None=CameraVectorWS()) {transformSourceType=TRANSFORMSOURCE_World, transformType=TRANSFORM_Tangent}
  n2 = ComponentMask(None=@n1) {g=true, r=true}
  n3 = ComponentMask(None=@n1) {b=true}
  MaterialExpressionAdd_0 = Add(
   A=Multiply(A=Divide(A=@n2, B=@n3), B=ScalarParameter() {defaultValue=0.5, parameterName=asdf})
   B=ComponentMask(None=VectorParameter() {parameterName=a, defaultValue=<rgba 1 0 0 0>}.RGB) {g=true, r=true}
  )
  MaterialExpressionScalarParameter_1 = ScalarParameter() {defaultValue=2, parameterName=BLurDistance}
  MaterialExpressionScalarParameter_2 = ScalarParameter() {defaultValue=32, parameterName=Steps}
  Custom!(
   Tex=@MaterialExpressionTextureObject_1
   UVW=@MaterialExpressionTextureCoordinate_0
   Vector=@MaterialExpressionAdd_0
   Distance=@MaterialExpressionScalarParameter_1
   Steps=@MaterialExpressionScalarParameter_2
  ) {
   code="\r\n\r\nfloat3 accum = 0;\r\nfloat2 curUVW = UVW;\r\n\r\nfloat stepsize = Distance / Steps;\r\nstepsize /=10;\r\n\r\nfor(int i =0; i < Steps; i++)\r\n{\r\n\taccum += Tex.SampleLevel(TexSampler, curUVW, 0);\r\n\t\r\n\tcurUVW += Vector * stepsize;\r\n}\r\n\r\nreturn accum/Steps;"
   inputs [
    {inputName=Tex, input {expression=&self:MaterialExpressionTextureObject_1}}
    {inputName=UVW, input {expression=&self:MaterialExpressionTextureCoordinate_0}}
    {inputName=Vector, input {expression=&self:MaterialExpressionAdd_0}}
    {inputName=Distance, input {expression=&self:MaterialExpressionScalarParameter_1}}
    {inputName=Steps, input {expression=&self:MaterialExpressionScalarParameter_2}}
   ]
  }
  TextureSample() {texture=&flipbook}
 ]
}
