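# Debug material that displays a single Gerstner water wave with parameters for direction, wavelength, amplitude and steepness.
# Two Custom HLSL nodes share those parameters, world XY and time, using gravity-based dispersion to compute a world position offset and an analytic normal.
# The normal is rebuilt for world-space use with Z inverted from one, and its X component feeds both X and Y.
# A Base vector parameter supplies a dark blue color.
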Material /Water/Materials/Debug/Single_Gerstner_Wave
graph {
 properties {bTangentSpaceNormal=false, pixelDepthOffsetMode=PDOM_Legacy, usedShadingModels=Default Lit}
 nodes [
  MaterialExpressionComponentMask_0 = ComponentMask(None=WorldPosition() {worldPositionShaderOffset=WPT_ExcludeAllShaderOffsets}.XYZ) {g=true, r=true}
  MaterialExpressionNormalize_0 = Normalize(VectorInput=ComponentMask(None=VectorParameter() {parameterName=Wave Direction, defaultValue=<rgba 1 0 0 0>}.RGB) {g=true, r=true})
  MaterialExpressionScalarParameter_0 = ScalarParameter() {defaultValue=2048, parameterName=WaveLength}
  MaterialExpressionScalarParameter_1 = ScalarParameter() {defaultValue=256, parameterName=Amplitude}
  MaterialExpressionScalarParameter_3 = ScalarParameter() {parameterName=Steepness}
  MaterialExpressionTime_0 = Time()
  Custom!(
   WorldPos=@MaterialExpressionComponentMask_0
   WaveDir=@MaterialExpressionNormalize_0
   Wavelength=@MaterialExpressionScalarParameter_0
   Amplitude=@MaterialExpressionScalarParameter_1
   Steepness=@MaterialExpressionScalarParameter_3
   Time=@MaterialExpressionTime_0
  ) {
   code="\r\n\r\nconst float gravity = 980;\r\n\r\nfloat dispersion = 2 * PI / Wavelength;\r\nfloat2 wavevector = WaveDir * dispersion;\r\nfloat wavespeed = sqrt(dispersion * gravity);\r\nfloat wavetime = wavespeed * Time;\r\n\r\nfloat wavepos = dot(WorldPos, wavevector) - wavetime;\r\n\r\nfloat wavesin = sin(wavepos);\r\nfloat wavecos = cos(wavepos);\r\n\r\nfloat wKA = Amplitude * dispersion;\r\n\r\nfloat q = Amplitude * (Steepness / wKA);\r\n\r\nfloat3 offset = 0;\r\noffset.xy = -q * wavesin * WaveDir;\r\noffset.z = wavecos * Amplitude;\r\n\r\nreturn offset;\r\n\r\n\r\n\r\n"
   description=Gerstner Wave WPO
   inputs [
    {inputName=WorldPos, input {expression=&self:MaterialExpressionComponentMask_0}}
    {inputName=WaveDir, input {expression=&self:MaterialExpressionNormalize_0}}
    {inputName=Wavelength, input {expression=&self:MaterialExpressionScalarParameter_0}}
    {inputName=Amplitude, input {expression=&self:MaterialExpressionScalarParameter_1}}
    {inputName=Steepness, input {expression=&self:MaterialExpressionScalarParameter_3}}
    {inputName=Time, input {expression=&self:MaterialExpressionTime_0}}
   ]
  }
  n1 = Custom(
   WorldPos=@MaterialExpressionComponentMask_0
   WaveDir=@MaterialExpressionNormalize_0
   Wavelength=@MaterialExpressionScalarParameter_0
   Amplitude=@MaterialExpressionScalarParameter_1
   Steepness=@MaterialExpressionScalarParameter_3
   Time=@MaterialExpressionTime_0
  ) {
   code="\r\n\r\nconst float gravity = 980;\r\n\r\nfloat dispersion = 2 * PI / Wavelength;\r\nfloat2 wavevector = WaveDir * dispersion;\r\nfloat wavespeed = sqrt(dispersion * gravity);\r\nfloat wavetime = wavespeed * Time;\r\n\r\nfloat wavepos = dot(WorldPos, wavevector) - wavetime;\r\n\r\nfloat wavesin = sin(wavepos);\r\nfloat wavecos = cos(wavepos);\r\n\r\nfloat wKA = Amplitude * dispersion;\r\n\r\nfloat q = (Steepness / wKA);\r\n\r\nfloat3 normal = 0;\r\n\r\nnormal.xy = wavesin * wKA * WaveDir;\r\nnormal.z = wavecos * wKA * q;\r\n\r\nreturn normal;"
   description=Gerstner Wave Normals
   inputs [
    {inputName=WorldPos, input {expression=&self:MaterialExpressionComponentMask_0}}
    {inputName=WaveDir, input {expression=&self:MaterialExpressionNormalize_0}}
    {inputName=Wavelength, input {expression=&self:MaterialExpressionScalarParameter_0}}
    {inputName=Amplitude, input {expression=&self:MaterialExpressionScalarParameter_1}}
    {inputName=Steepness, input {expression=&self:MaterialExpressionScalarParameter_3}}
    {inputName=Time, input {expression=&self:MaterialExpressionTime_0}}
   ]
  }
  n2 = MaterialFunctionCall(Float3=@n1) {materialFunction=&BreakOutFloat3Components}
  MaterialFunctionCall(X=@n2.R, Y=@n2.R, Z=OneMinus(None=@n2.B)) {materialFunction=&MakeFloat3}
  VectorParameter() {parameterName=Base, defaultValue=<rgba 0 0.04 0.1 0>}
 ]
}
